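; Test Blueprint for a kitchen pro range prop that changes its look for a Season 38 story quest.
; On a side effect event carrying the Story.P03 quest tag, it swaps the grilles mesh to the cooking-veggies grill.
; The swap runs only on the Remote (non-authority) branch and prints a placeholder debug string with the SetStaticMesh result.
; BeginPlay, Tick and overlap events are empty.
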(event EventBeginPlay)

(event Collision|EventActorBeginOverlap
  (params (OtherActor :type Actor%20Object%20Reference)))

(event EventTick
  (params (DeltaSeconds :type Float%20%28single-precision%29)))

(event SideEffects|EventBP_ApplySideEffect
  (params (ReactionSource :type Object%20Reference) (TargetObject :type Object%20Reference) (SideEffectTriggerTag :type Gameplay%20Tag%20Structure%20%28by%20ref%29) (SideEffectTriggerType :type EBlueprintableSideEffectTriggerType%20Enum%20%28by%20ref%29 :default "TriggeredAsTarget"))
  (Networking|SwitchHasAuthority
    (:Remote
      (if (== SideEffectTriggerTag "(TagName=\"Athena.Quests.S38.CTA.Story.P03\")")
        (bind _returnvalue (Components|StaticMesh|SetStaticMesh (Variables|Default|Getgrilles) "/WildEstate/Environment/Props/Quest/Quest_CookingVeggies_A/Meshes/SM_Quest_CookingVeggiesGrill_A.SM_Quest_CookingVeggiesGrill_A"))
        (Development|PrintString (Utilities|String|Append "Testy McTesting" (Utilities|String|ToString(Boolean) _returnvalue)) true true "(R=0.000000,G=0.660000,B=1.000000,A=1.000000)" 1e+01)))))

(fn ConstructionScript
  (|CallParentFunction))
